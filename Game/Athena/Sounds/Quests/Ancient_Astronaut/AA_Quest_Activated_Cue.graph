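# Sound cue for the Ancient Astronaut quest activation sound.
# A mixer layers the quest objective granted wave at full volume with the part retrieved wave at half volume.
# Its own attenuation override replaces the referenced gadget activate settings, adding spatial falloff, occlusion with low-pass filtering, and listener focus.
# It is routed through a third-person ranged weapon fire sound class rather than a quest-specific one.
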SoundCue /Game/Athena/Sounds/Quests/Ancient_Astronaut/AA_Quest_Activated_Cue
root=Mixer_0
nodes [
 {id=Mixer_0, children [WavePlayer_1,WavePlayer_0], properties {inputVolume [1,0.5]}}
 {id=WavePlayer_0, properties {soundWaveAssetPtr=&/AA_Part_Retrieved}}
 {id=WavePlayer_1, properties {soundWaveAssetPtr=&/AA_Quest_Objective_Granted}}
]
properties {
 attenuationSettings=&Fort_GadgetActivate_Att
 bOverrideAttenuation=true
 maxDistance=6000
 soundClassObject=&Fort_WeaponRanged_Fire_3P_PSM
 attenuationOverrides {
  bEnableListenerFocus=true
  bEnableOcclusion=true
  falloffDistance=5000
  nonFocusVolumeAttenuation=0.3
  occlusionLowPassFilterFrequency=10000
  occlusionTraceChannel=ECC_GameTraceChannel11
  occlusionVolumeAttenuation=0.5
  stereoSpread=500
  attenuationShapeExtents {x=1000}
 }
}
